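; Passive gameplay ability for the Deft Hands player augment that auto-reloads a specific weapon when the player switches away from it.
; On activation it calls the parent and, if the AutoReloadEnabled scalable float is true, binds to the pawn's weapon-equipped delegate.
; On each switch it compares the previous weapon's item definition to the first ItemsToGrant entry and, on a match, tries to reload that holstered weapon.
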(event Ability|EventFailedToActivatePassiveAbility
  (params (ActorInfo :type Gameplay%20Ability%20Actor%20Info%20Structure)))

(event Ability|EventActivateAbility
  (|CallParentFunction)
  (if (ScalableFloat|GetValueasBool (Variables|Default|GetAutoReloadEnabled))
    (FortPawn|BindEventtoOnWeaponEquippedDelegate (Ability|GetActivatingPawn) (delegate Custom|OnWeaponSwitched))))

(event Custom|OnWeaponSwitched
  (params (NewWeapon :type Fort%20Weapon%20Object%20Reference) (PrevWeapon :type Fort%20Weapon%20Object%20Reference))
  (bind _returnvalue (FortWeapon|GetInventoryItem PrevWeapon))
  (bind _output_itemdefinition_2_8f1a360a4a0335351ef6d789295de464 (Utilities|Struct|BreakBFPlayerAugmentGrantedItemData (Utilities|Array|Get(acopy) (Variables|Default|GetItemsToGrant))))
  (if (Utilities|IsValid _returnvalue :!class /Script/Engine.KismetSystemLibrary)
    (if (== (Item|GetItemDefinition _returnvalue) _output_itemdefinition_2_8f1a360a4a0335351ef6d789295de464)
      (FortWeapon|TryReloadAmmofromAbility PrevWeapon))))
